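; Conversation side effect node that, on the server, converts the NPC in the conversation on behalf of the talking player.
; ServerCauseSideEffect calls the parent, then passes the player pawn and NPC pawn into TriggerConversion.
; GetNPC casts the NPC participant actor to a Fort pawn, while GetPlayer casts the player participant to a PlayerController and resolves its Fort pawn through the quest library.
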(fn GetNPC
  (params (Context :type Conversation%20Context%20Structure))
  (bind _returnvalue (Class|FortConversationContextLibrary|GetParticipantActorNPC Context))
  (Utilities|IsValid _returnvalue
    (:Is%20Valid
      (bind _asfort_pawn (Utilities|Casting|CastToFortPawn _returnvalue))
      (return _asfort_pawn))))

(fn GetPlayer
  (params (Context :type Conversation%20Context%20Structure))
  (bind _returnvalue (Class|FortConversationContextLibrary|GetParticipantActorPlayer Context))
  (Utilities|IsValid _returnvalue
    (:Is%20Valid
      (bind _asplayer_controller (Utilities|Casting|CastToPlayerController _returnvalue))
      (bind _fortpawn (Class|FortQuestBlueprintLibrary|GetFortPawnFromObject _asplayer_controller)
        (:Valid
          (return _fortpawn))))))

(fn ServerCauseSideEffect
  (params (Context :type Conversation%20Context%20Structure%20%28by%20ref%29))
  (|CallParentFunction Context)
  (bind _fortpawn (CallFunction|GetNPC Context))
  (bind _fortpawn_2 (CallFunction|GetPlayer Context))
  (Convert|TriggerConversion _fortpawn_2 _fortpawn)
  (return))
