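; Conversation side effect node that removes a tagged item from the talking player's inventory.
; ServerCauseSideEffect calls the parent, resolves the participant player from the conversation context and casts it to a Fort Player Controller.
; RemoveItemFromInventory finds the item by the configured tag, removes one by GUID, then sends an Event.Item.Dropped gameplay event to the controller.
; It returns whether an item was found, though the side effect ignores that result.
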(fn RemoveItemFromInventory
  (params (PlayerController :type Fort%20Player%20Controller%20Object%20Reference))
  ;[0 0 0] Remove the item with the specified tag from the player's inventory
  (bind _returnvalue (Inventory|GetInventoryItemwithTag PlayerController (Variables|Default|GetItem%20to%20Remove%20Tag)))
  (Utilities|IsValid _returnvalue
    (:Is%20Valid
      (Inventory|RemoveItemFromPlayerByGuid PlayerController (Item|GetItemGuid _returnvalue) 1)
      (Ability|SendGameplayEventtoActor PlayerController "(TagName=\"Event.Item.Dropped\")" (Utilities|Struct|MakeGameplayEventData "(TagName=\"Event.Item.Dropped\")" PlayerController 0 _returnvalue))
      (return true))
    (:Is%20Not%20Valid
      (return false))))

(fn ServerCauseSideEffect
  (params (Context :type Conversation%20Context%20Structure%20%28by%20ref%29))
  (|CallParentFunction Context)
  (bind _returnvalue (Class|FortConversationContextLibrary|GetParticipantActorPlayer Context))
  (Utilities|IsValid _returnvalue
    (:Is%20Valid
      (bind _asfort_player_controller (Utilities|Casting|CastToFortPlayerController _returnvalue))
      (Default|RemoveItemfromInventory _asfort_player_controller)
      (return))))
